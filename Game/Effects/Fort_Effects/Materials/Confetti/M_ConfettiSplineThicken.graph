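# Unlit, masked, two-sided confetti material for Niagara mesh particles, turning a spline mesh into swirling streamers.
# A vertex offset widens the strip across a UV-rotated axis, adds time-panned curl noise growing over particle life, nudges along the camera vector, and optionally scales by dynamic Width.
# A gradient mask travels along V with particle age.
# Particle-colored emissive dims with offset distance and differs per face side.
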Material /Game/Effects/Fort_Effects/Materials/Confetti/M_ConfettiSplineThicken
graph {
 properties {
  bTangentSpaceNormal=false
  bUsedWithNiagaraMeshParticles=true
  blendMode=BLEND_Masked
  pixelDepthOffsetMode=PDOM_Legacy
  shadingModel=MSM_Unlit
  twoSided=true
  usedShadingModels=Unlit
 }
 nodes [
  n1 = Distance(A=WorldPosition() {worldPositionShaderOffset=WPT_ExcludeAllShaderOffsets}.XYZ, B=WorldPosition().XYZ)
  n2 = ScalarParameter() {defaultValue=72, parameterName=Length}
  n3 = OneMinus(None=Saturate(None=Divide(A=@n1, B=@n2)))
  n4 = ParticleColor()
  n5 = TwoSidedSign()
  n6 = OneMinus(None=Multiply(A=Saturate(None=@n5)) {constB=0.5})
  MaterialFunctionCall(Emissive=Multiply(A=Power(Base=@n3) {constExponent=3}, B=Multiply(A=@n4.RGB, B=@n6))) {materialFunction=&MF_FX_EmissiveAdjustments}
  MaterialFunctionCall!(
   Gradient=ComponentMask(None=TextureCoordinate()) {g=true}
   Mask Position=ParticleRelativeTime()
   Mask Size=ScalarParameter() {defaultValue=0.333, parameterName=Mask Size}
  ) {materialFunction=/Game/Effects/Fort_Effects/MaterialFunctions/MF_GradientMask}
  n7 = VertexNormalWS()
  n8 = Normalize(VectorInput=Transform(None=Constant3Vector() {constant=<rgba 0 1 0 0>}))
  n9 = Multiply(A=ComponentMask(None=TextureCoordinate()) {g=true}, B=ScalarParameter() {defaultValue=3, parameterName=Swirls})
  n10 = Constant()
  n11 = CrossProduct(A=Add(A=@n7, B=RotateAboutAxis(NormalizedRotationAxis=@n8, RotationAngle=@n9, PivotPoint=@n10, Position=@n7)), B=@n8)
  n12 = Add(A=@n4.A, B=ScalarParameter() {defaultValue=2, parameterName=Width})
  n13 = Subtract(A=ComponentMask(None=TextureCoordinate()) {r=true}) {constB=0.5}
  n14 = WorldPosition()
  n15 = Add(A=Time(), B=Multiply(A=ParticleRandom()) {constB=222})
  n16 = ScalarParameter() {defaultValue=0.5, parameterName=Curl Noise Pan Speed}
  n17 = Add(
   A=Divide(
    A=TransformPosition(None=@n14.XYZ) {transformSourceType=TRANSFORMPOSSOURCE_World, transformType=TRANSFORMPOSSOURCE_Instance}
    B=ScalarParameter() {defaultValue=22, parameterName=Curl Noise Scale}
   )
   B=AppendVector(A=Constant2Vector(), B=Multiply(A=@n15, B=@n16))
  )
  n18 = ScalarParameter() {defaultValue=4, parameterName=Curl Noise Intensity}
  n19 = ParticleRelativeTime()
  n20 = Add(
   A=Multiply(A=Normalize(VectorInput=@n11), B=Multiply(A=@n12, B=@n13))
   B=Multiply(A=VectorNoise(World Position=@n17) {noiseFunction=VNF_CurlALU}, B=Multiply(A=@n18, B=@n19))
  )
  n21 = Subtract(A=Add(A=WorldPosition().XYZ, B=@n20), B=CameraPositionWS())
  n22 = Add(A=Multiply(A=@n4.A, B=Normalize(VectorInput=@n21)), B=@n20)
  n23 = DynamicParameter() {paramNames [Width,Param2,Param3,Param4]}
  Reroute(None=LinearInterpolate(A=@n22, B=Multiply(A=@n23.Width, B=@n22), Alpha=ScalarParameter() {parameterName=UseDynParameter}))
 ]
}
