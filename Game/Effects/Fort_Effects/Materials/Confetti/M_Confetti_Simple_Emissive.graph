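# Translucent unlit particle material for simple emissive confetti on Cascade and Niagara sprites.
# A time sine, sped by per-particle random, pulses particle color brightness for twinkle, then a random hue shift varies it, blended in by a scalar parameter.
# Mobile uses plain particle color; both paths feed a shared emissive adjustment function.
# A selectable channel of the 4x4 confetti SubUV texture, multiplied by particle alpha, forms the alpha mask.
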Material /Game/Effects/Fort_Effects/Materials/Confetti/M_Confetti_Simple_Emissive
graph {
 properties {
  bUsedWithNiagaraSprites=true
  bUsedWithParticleSprites=true
  blendMode=BLEND_Translucent
  pixelDepthOffsetMode=PDOM_Legacy
  shadingModel=MSM_Unlit
  translucencyLightingMode=TLM_VolumetricPerVertexDirectional
  usedShadingModels=Unlit
 }
 nodes [
  n1 = ParticleColor()
  n2 = ParticleRandom()
  n3 = ScalarParameter() {defaultValue=3, parameterName=TwinkleSpeed}
  n4 = Multiply(A=MaterialFunctionCall(Speed=Multiply(A=@n2, B=@n3)) {materialFunction=&DebugTimeSine}.Result) {constB=5}
  n5 = MaterialFunctionCall(Hue Shift Percentage=Multiply(A=@n2) {constB=0.8}, Texture=Multiply(A=Clamp(None=@n4) {maxDefault=5, minDefault=0.25}, B=@n1.RGB)) {materialFunction=&HueShift}
  n6 = ScalarParameter() {defaultValue=1, parameterName=Use Particle Random Color}
  MaterialFunctionCall(Emissive=ShadingPathSwitch(Default=LinearInterpolate(A=@n1.RGB, B=@n5.Result, Alpha=@n6), Mobile=@n1.RGB) {desc=Migrated from FeatureLevelSwitch}) {materialFunction=&MF_FX_EmissiveAdjustments}
  Multiply(
   A=ParticleColor().A
   B=StaticComponentMaskParameter(None=TextureSample() {texture=/Game/Effects/Fort_Effects/Textures/Confetti/T_Confetti_SubUV_4X4_A}.B) {defaultB=true, parameterName=PickChannel}
  )
 ]
}
