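# Unlit, translucent, two-sided trail material in the confetti effects set, built for beam trail particles and colored by particle color.
# A side mask drives center-based V scaling of UV1, varying the ribbon pattern's apparent width, before adjustable tiling and looping time-based panning.
# The ribbon texture's green channel sets brightness for emissive adjustment, and its alpha multiplies particle alpha.
# Metallic, specular and roughness parameters are present but unconnected.
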Material /Game/Effects/Fort_Effects/Materials/Confetti/M_Confetti_Trail
graph {
 properties {
  bUsedWithBeamTrails=true
  blendMode=BLEND_Translucent
  pixelDepthOffsetMode=PDOM_Legacy
  shadingModel=MSM_Unlit
  twoSided=true
  usedShadingModels=Unlit
 }
 nodes [
  n1 = ParticleColor()
  n2 = Constant() {r=0.1}
  n3 = Constant() {r=1}
  n4 = TextureSample() {texture=/Game/Effects/Fort_Effects/Textures/Masks/T_Side_Mask}
  n5 = MaterialFunctionCall(UVs=TextureCoordinate() {coordinateIndex=1}, Texture Scale=AppendVector(A=Constant() {r=1}, B=LinearInterpolate(A=@n2, B=@n3, Alpha=@n4.G))) {materialFunction=&ScaleUVsByCenter}
  n6 = AppendVector(
   A=ScalarParameter() {defaultValue=1, group=Tiling Tex, parameterName=Scale U}
   B=ScalarParameter() {defaultValue=1, group=Tiling Tex, parameterName=Scale V}
  )
  n7 = AppendVector(A=ScalarParameter() {defaultValue=0.5, group=Tiling Tex, parameterName=Pan U}, B=ScalarParameter() {group=Tiling Tex, parameterName=Pan V})
  n8 = Time() {bOverride_Period=true, period=2}
  n9 = TextureSample(UVs=Add(A=Multiply(A=@n5.UVs, B=@n6), B=Multiply(A=@n7, B=@n8))) {texture=/Game/Gadgets/Assets/VinderTech_GliderChute/Glider_TeddyBear/Effects/Textures/T_RibbonTrail_Tile}
  n10 = LinearInterpolate(A=ScalarParameter() {defaultValue=0.25, parameterName=A}, B=ScalarParameter() {defaultValue=1, parameterName=B}, Alpha=@n9.G)
  MaterialFunctionCall(Emissive=Multiply(A=Multiply(A=@n1.RGB, B=@n10), B=ScalarParameter() {defaultValue=1, group=Surface, parameterName=EmissiveIntensity})) {materialFunction=&MF_FX_EmissiveAdjustments}
  Multiply(A=@n1.A, B=@n9.A)
  ScalarParameter() {group=Surface, parameterName=Met}
  ScalarParameter() {group=Surface, parameterName=Spec}
  ScalarParameter() {defaultValue=0.5, group=Surface, parameterName=Rough}
 ]
}
