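# Translucent, two-sided lit sprite material for GPU paper confetti pieces, named for the Egg Launcher effect.
# Per-particle randomness varies particle color between half and full brightness and blends UV tilings for beam and smoke-trail textures.
# A doubled smoke-trail channel scales particle alpha into a mask, while RotateAboutAxis computes a time-driven tumble around each particle's position on a randomized axis.
# Material output connections are not shown, and several texture and transform nodes feed nothing visible.
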Material /Game/Effects/Fort_Effects/Materials/Confetti/M_EggLauncher_GreenpPaper_Pieces_GPU
graph {
 properties {bUsedWithParticleSprites=true, blendMode=BLEND_Translucent, pixelDepthOffsetMode=PDOM_Legacy, twoSided=true, usedShadingModels=Default Lit}
 nodes [
  Constant() {r=0.25}
  Constant()
  n1 = ParticleColor()
  LinearInterpolate(A=Multiply(B=@n1.RGB) {constA=0.5}, B=@n1.RGB, Alpha=ParticleRandom())
  n2 = TextureCoordinate() {uTiling=2}
  n3 = TextureCoordinate() {uTiling=0.5}
  n4 = ParticleRandom()
  Multiply(A=TextureSample(UVs=LinearInterpolate(A=@n2, B=@n3, Alpha=@n4)) {samplerType=SAMPLERTYPE_LinearColor, texture=/Game/Effects/Fort_Effects/Textures/Tile/T_Raygun_Beam_01}.R)
  n5 = LinearInterpolate(A=TextureCoordinate() {uTiling=0.25}, B=TextureCoordinate(), Alpha=ParticleRandom())
  n6 = TextureSample(UVs=@n5) {texture=/Game/Effects/Fort_Effects/Textures/Smoke/T_Smoke_ThinTrail_Packed}
  Multiply(A=@n1.A, B=Add(A=@n6.B, B=@n6.B))
  ObjectPositionWS()
  RotateAboutAxis(
   NormalizedRotationAxis=LinearInterpolate(A=Constant3Vector() {constant=<rgba 0 1 0.6 0>}, B=Constant3Vector() {constant=<rgba 1 0.25 0 0>}, Alpha=ParticleRandom())
   RotationAngle=Multiply(A=Time(), B=ParticleRandom())
   PivotPoint=ParticlePositionWS()
   Position=WorldPosition().XYZ
  )
  TextureSample(UVs=@n2) {samplerType=SAMPLERTYPE_LinearColor, texture=/Game/Effects/Fort_Effects/Textures/Tile/T_Raygun_Beam_02}
  TextureSample(UVs=@n5) {samplerType=SAMPLERTYPE_Grayscale, texture=&T_Smoke_Trail_Soft}
  n7 = WorldPosition()
  TransformPosition(None=@n7.XYZ) {transformType=TRANSFORMPOSSOURCE_World}
  Transform(None=@n7.XYZ) {transformSourceType=TRANSFORMSOURCE_World, transformType=TRANSFORM_Local}
 ]
}
